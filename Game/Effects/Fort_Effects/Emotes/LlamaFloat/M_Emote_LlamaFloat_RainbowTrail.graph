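# Translucent, unlit, two-sided rainbow trail material for the LlamaFloat emote, usable on Niagara ribbons, mesh particles and skeletal meshes.
# The default look tints a swizzled rainbow texture by particle color, with opacity from edge masking, depth, camera, fresnel and view-angle fades.
# Scalar-driven lerps switch both color and opacity to a skeletal tail (llama emissive texture, UV/sphere masks) or a center mask (rainbow through a ripple texture).
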Material /Game/Effects/Fort_Effects/Emotes/LlamaFloat/M_Emote_LlamaFloat_RainbowTrail
graph {
 properties {
  bUsedWithMeshDeformer=true
  bUsedWithNiagaraMeshParticles=true
  bUsedWithNiagaraRibbons=true
  bUsedWithSkeletalMesh=true
  blendMode=BLEND_Translucent
  pixelDepthOffsetMode=PDOM_Legacy
  shadingModel=MSM_Unlit
  twoSided=true
  usedShadingModels=Unlit
 }
 nodes [
  n1 = TextureCoordinate()
  n2 = Add(A=OneMinus(None=ComponentMask(None=@n1) {r=true}), B=ScalarParameter() {defaultValue=-0.33, group=SphereMask, parameterName=Mask_offset}) {constB=-0.33}
  n3 = ScalarParameter() {defaultValue=2, group=SphereMask, parameterName=Power}
  n4 = Saturate(None=Multiply(A=Power(Base=@n2, Exp=@n3), B=ScalarParameter() {defaultValue=1000, group=SphereMask, parameterName=Edge hardness}) {constB=1000})
  n5 = VectorParameter() {group=SphereMask, parameterName=UV_Scale, channelNames {a="", b="", g=V, r=U}, defaultValue=<rgba 1 1 0 0>}
  n6 = SphereMask(
   A=Multiply(A=ComponentMask(None=@n5) {g=true, r=true}, B=TextureCoordinate())
   B=ComponentMask(None=VectorParameter() {group=SphereMask, parameterName=SphereMask_Offset, defaultValue=<rgba 0.49 0.82 0 1>}.RGB) {g=true, r=true}
   Radius=ScalarParameter() {defaultValue=0.2, group=SphereMask, parameterName=Radius}
  ) {attenuationRadius=0.2, hardnessPercent=0}
  n7 = ScalarParameter() {defaultValue=2, group=SphereMask, parameterName=SphereMask_Power}
  n8 = Saturate(None=Multiply(A=Power(Base=@n6, Exp=@n7), B=ScalarParameter() {defaultValue=100, group=SphereMask, parameterName=SphereMask_Edge_hardness}) {constB=100})
  n9 = ParticleColor()
  n10 = ComponentMask(None=TextureCoordinate()) {g=true}
  n11 = Multiply(A=@n10, B=OneMinus(None=@n10))
  n12 = Multiply(
   A=@n9.A
   B=MaterialFunctionCall(In=Clamp(None=@n11), Contrast=ScalarParameter() {parameterName=MaskContrast}) {materialFunction=&CheapContrast}.Result
  )
  n13 = ScalarParameter() {parameterName=UseAlphaMAsk}
  n14 = MaterialFunctionCall(Distance=ScalarParameter() {defaultValue=100, group=Alpha, parameterName=NearCameraFade}) {materialFunction=&Near_Camera_Fade}
  n15 = ScalarParameter() {defaultValue=35, group=Alpha, parameterName=FadeDistance}
  n16 = Multiply(
   A=Multiply(A=LinearInterpolate(A=@n9.A, B=@n12, Alpha=@n13), B=DepthFade(Opacity=@n14.Result, FadeDistance=@n15) {fadeDistanceDefault=50})
   B=MaterialFunctionCall(Axis=ParticleDirection(), Exponent=ScalarParameter() {defaultValue=1, parameterName=Fresnel Exponent}) {materialFunction=&AxisAlignedFresnel}.Fresnel
  )
  n17 = MaterialFunctionCall("FadeSensitivity (optional)"=ScalarParameter() {parameterName=FadeBasedOnViewAngle}) {materialFunction=&FadeBasedOnViewAngle}
  n18 = TextureSample() {samplerType=SAMPLERTYPE_LinearColor, texture=/Game/Effects/Fort_Effects/Effects/Water/Textures/WaterDecalRippleMask}
  n19 = ParticleColor()
  n20 = ScalarParameter() {defaultValue=3, parameterName=Angle_Fade}
  LinearInterpolate(
   A=LinearInterpolate(A=Multiply(A=@n16, B=@n17.Result), B=Multiply(A=@n4, B=@n8), Alpha=ScalarParameter() {parameterName=Only_Use_Skeletal_Mesh_tail})
   B=Multiply(
    A=Multiply(A=@n18.R, B=@n19.A)
    B=MaterialFunctionCall("FadeSensitivity (optional)"=@n20) {materialFunction=&FadeBasedOnViewAngle}.Result
   )
   Alpha=ScalarParameter() {parameterName=Only_Use_Center_Mask}
  )
  n21 = ScalarParameter() {parameterName=Only_Use_Skeletal_Mesh_tail}
  n22 = Multiply(A=Desaturation(None=TextureSample() {texture=/Game/Animation/Props/Llama_Float/Textures/T_Emote_Prop_Llama_Float_E}.RGB, Fraction=Constant())) {constB=2}
  n23 = TextureCoordinate()
  n24 = Add(
   A=Multiply(
    A=MaterialFunctionCall(XY=@n23) {materialFunction=&Swizzle}.YX
    B=ScalarParameter() {defaultValue=1, parameterName=UV_Tiling}
   ) {constB=2}
   B=ComponentMask(None=VectorParameter() {parameterName=UV_offset, defaultValue=<rgba 0 0 0 1>}.RGB) {g=true, r=true}
  )
  n25 = Multiply(
   A=ParticleColor().RGB
   B=Desaturation(
    None=TextureSampleParameter2D(UVs=@n24) {parameterName=Default_Tex, texture=&/T_LlamaFloat_Rainbow}.RGB
    Fraction=ScalarParameter() {defaultValue=-1, parameterName=Desaturation}
   )
  )
  n26 = Multiply(A=@n19.RGB, B=@n18.R)
  n27 = TextureSampleParameter2D(UVs=MaterialFunctionCall(XY=TextureCoordinate() {vTiling=-1}) {materialFunction=&Swizzle}.YX) {parameterName=Default_Tex, texture=&T_LlamaFloat_Rainbow}
  MaterialFunctionCall(Emissive=LinearInterpolate(
    A=LinearInterpolate(A=@n25, B=@n22, Alpha=@n21)
    B=Multiply(A=@n26, B=@n27.RGB)
    Alpha=ScalarParameter() {parameterName=Only_Use_Center_Mask}
   )) {materialFunction=&MF_FX_EmissiveAdjustments}
  Reroute(None=Multiply(A=TextureSample() {texture=/Game/Animation/Props/Llama_Float/Textures/T_Emote_Prop_Llama_Float_D}.RGB) {constB=0.5})
  Transform(None=Constant3Vector() {constant=<rgba 0 1 0 0>}) {transformType=TRANSFORM_Local}
 ]
}
